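# Translucent, unlit, two-sided material for spline ribbon FX on the OceanBreeze backpack, built for Niagara mesh particles.
# SplineThicken sets width from a curve-atlas profile, while two noise layers swing the ribbon perpendicular to the camera, scaled by dynamic parameters.
# A panning silk texture, with optional noise distortion and second layer, drives particle-tinted emissive and opacity.
# Edge masks, a facing-based fade from the SplineThicken normal, depth fade and particle alpha soften it.
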Material /Game/Accessories/FORT_Backpacks/Backpack_M_MED_OceanBreeze/FX/M_Spline_OceanBreeze
graph {
 properties {
  bUsedWithNiagaraMeshParticles=true
  blendMode=BLEND_Translucent
  pixelDepthOffsetMode=PDOM_Legacy
  shadingModel=MSM_Unlit
  twoSided=true
  usedShadingModels=Unlit
 }
 nodes [
  n1 = DynamicParameter() {paramNames [WPO, DF, "RibbonBScale(Special)", Param4]}
  n2 = CameraPositionWS()
  n3 = WorldPosition()
  n4 = Normalize(VectorInput=CrossProduct(A=Subtract(A=@n2, B=@n3.XYZ), B=MaterialFunctionCall() {materialFunction=&TangentBasis}.Y V))
  n5 = MaterialFunctionCall(Float2=NamedRerouteUsage()) {materialFunction=&BreakOutFloat2Components}
  n6 = CurveAtlasRowParameter(CurveTime=@n5.R) {
   atlas=&/CLA_OceanBreeze_A
   curve=&/CL_OceanBreeze_Spline
   parameterName=RibbonCurve
  }
  n7 = ScalarParameter() {parameterName=NonUseCurve}
  n8 = TextureObjectParameter() {
   group=SwingWPO
   parameterName=SwingWPO_Texture
   samplerType=SAMPLERTYPE_LinearColor
   sortPriority=0
   texture=/Game/Global/Textures/Default/Tileable_Patterns/T_PerlinNoise_Uncompressed
  }
  n9 = VectorParameter() {group=SwingWPO, parameterName=SwingWPO_UVControl, channelNames {a=VPan, b=UPan, g=VScale, r=UScale}, defaultValue=<rgba 1.5 1 0.25 0>}
  n10 = NamedRerouteUsage()
  n11 = AppendVector(
   A=MaterialFunctionCall(Float2=NamedRerouteUsage()) {materialFunction=&BreakOutFloat2Components}.R
   B=Constant()
  )
  n12 = ScalarParameter() {defaultValue=0.5, parameterName=SwingWPOOffset}
  n13 = Subtract(
   A=MaterialFunctionCall(Texture=@n8, UVControl_RGB=@n9, UVControl_A=@n9, "[Override]TimeInput"=@n10, "[Override]UVInput"=@n11) {materialFunction=&FXCommon_TextureSample}.R
   B=@n12
  ) {constB=0.5}
  n14 = ScalarParameter() {defaultValue=5, group=SwingWPO, parameterName=SwingWPO_Strength}
  n15 = ScalarParameter() {defaultValue=2, group=SwingWPO, parameterName=SwingWPO2_Strength}
  n16 = VectorParameter() {group=SwingWPO, parameterName=SwingWPO2_UVControl, channelNames {a=VPan, b=UPan, g=VScale, r=UScale}, defaultValue=<rgba 3 1 0.25 0>}
  n17 = Subtract(
   A=MaterialFunctionCall(Texture=@n8, UVControl_RGB=@n16, UVControl_A=@n16, "[Override]TimeInput"=@n10, "[Override]UVInput"=@n11) {materialFunction=&FXCommon_TextureSample}.R
   B=@n12
  ) {constB=0.5}
  n18 = Multiply(A=LinearInterpolate(A=@n6.G, B=@n5.R, Alpha=@n7), B=Add(A=Multiply(A=@n13, B=@n14), B=Multiply(A=@n15, B=@n17)))
  n19 = ScalarParameter() {defaultValue=5, parameterName=Wdith}
  n20 = LinearInterpolate(A=@n6.R, B=@n5.R, Alpha=@n7)
  n21 = MaterialFunctionCall(WidthBase=Multiply(A=@n19, B=@n20), WidthTip=Multiply(A=@n19, B=@n20), FlattenPixelNormal=StaticBool() {value=true}) {materialFunction=&SplineThicken}
  Add(A=Multiply(A=@n1.WPO, B=Multiply(A=@n4, B=@n18)), B=@n21.WorldPositionOffset)
  n22 = VectorParameter() {group=RibbonA, parameterName=RibbonA_Color, defaultValue=<rgba 0 33.3334 200 0>}
  n23 = TextureObjectParameter() {
   group=RibbonA
   parameterName=RibbonA_Texture
   sortPriority=0
   texture=/Game/Gadgets/Assets/VinderTech_GliderChute/Glider_Emperor/FX/Texture/T_Emperor_Silk_01
  }
  n24 = VectorParameter() {group=RibbonA, parameterName=RibbonA_UVControl, channelNames {a=VPan, b=UPan, g=VScale, r=UScale}, defaultValue=<rgba 1 0.25 0.1 0>}
  n25 = NamedRerouteUsage()
  n26 = StaticBoolParameter() {defaultValue=true, parameterName=ScaleUVByCenter}
  n27 = VectorParameter() {group=NoiseA, parameterName=NoiseA_UVControl, channelNames {a=VPan, b=UPan, g=VScale, r=UScale}, defaultValue=<rgba 1 1 0 0>}
  n28 = MaterialFunctionCall(Texture=TextureObjectParameter() {group=NoiseA, parameterName=NoiseA_Texture, sortPriority=0}, UVControl_RGB=@n27, UVControl_A=@n27) {materialFunction=&FXCommon_TextureSample}
  n29 = AppendVector(A=@n28.R, B=@n28.R)
  n30 = VectorParameter() {
   group=NoiseA
   parameterName=NoiseA_NoiseControl
   sortPriority=60
   channelNames {a=VStrength, b=UStrength, g=VOffset, r=UOffset}
   defaultValue=<rgba -0.5 -0.5 0.01 0.01>
  }
  n31 = Add(
   A=NamedRerouteUsage()
   B=StaticSwitchParameter(
    True=MaterialFunctionCall!(NoiseValue=@n29, Control_RGB=@n30, Control_A=@n30) {materialFunction=/Game/Effects/Niagara/MasterMaterials/Functions/Common/FXCommon_NoiseControl}.NoiseValue
    False=Constant()
   ) {parameterName=UseNoiseA}
  )
  n32 = Saturate(None=Desaturation(
    None=MaterialFunctionCall!(
     Texture=@n23
     AdditionalUVOffset
     UVControl_RGB=@n24
     UVControl_A=@n24
     "[Override]TimeInput"=@n25
     "[Override]ScaleUVByCenter"=@n26
     "[Override]UVInput"=@n31
    ) {materialFunction=/Game/Effects/Niagara/MasterMaterials/Functions/Common/FXCommon_TextureSample_V2}.RGB
    Fraction=Constant() {r=1}
   ))
  n33 = ScalarParameter() {defaultValue=1, parameterName=OpacityPower}
  n34 = Saturate(None=Multiply(A=Power(Base=@n32, Exp=@n33), B=ScalarParameter() {defaultValue=1, parameterName=OpacityScale}))
  n35 = StaticSwitchParameter(
   True=LinearInterpolate(A=VectorParameter() {group=RibbonA, parameterName=RibbonB_Color, defaultValue=<rgba 0 33.3334 200 0>}.RGB, B=@n22.RGB, Alpha=@n34)
   False=@n22.RGB
  ) {parameterName=UseBTexture}
  n36 = ParticleColor()
  Multiply(
   A=MaterialFunctionCall(Emissive=Divide(A=@n35) {constB=1.9771}) {materialFunction=&MF_FX_EmissiveAdjustments}.Result
   B=@n36.RGB
  )
  n37 = Multiply(A=DotProduct(A=CameraVectorWS(), B=@n21.Normal), B=ScalarParameter() {defaultValue=2, parameterName=FresnelMul})
  n38 = MaterialFunctionCall(Float2=NamedRerouteUsage()) {materialFunction=&BreakOutFloat2Components}
  n39 = VectorParameter() {group=Opacity, parameterName=Opacity_EdgeMask_V_Control, channelNames {a="", b=Power, g=Multiply, r=Offset}, defaultValue=<rgba 0 5 1 1>}
  n40 = TextureObjectParameter() {
   group=RibbonB
   parameterName=RibbonB_Texture
   sortPriority=0
   texture=&/T_OceanBreeze_SplineB_t
  }
  n41 = VectorParameter() {group=RibbonB, parameterName=RibbonB_UVControl, channelNames {a=VPan, b=UPan, g=VScale, r=UScale}, defaultValue=<rgba 1 1 0 0>}
  n42 = MaterialFunctionCall(
   Texture=@n40
   UVControl_RGB=@n41
   UVControl_A=@n41
   "[Override]TimeInput"=NamedRerouteUsage()
   "[Override]ScaleUVByCenter"=StaticBoolParameter() {defaultValue=true, parameterName=BScaleUVByCenter}
   "[Override]UVInput"=@n31
  ) {materialFunction=&FXCommon_TextureSample_V2}
  n43 = Constant() {r=1}
  n44 = Power(Base=Saturate(None=Desaturation(None=@n42.RGB, Fraction=@n43)), Exp=ScalarParameter() {defaultValue=1, parameterName=BOpacityPower})
  n45 = ScalarParameter() {defaultValue=1, parameterName=BOpacityScale}
  n46 = LinearInterpolate(
   A=@n45
   B=Multiply(A=DynamicParameter() {paramNames [WPO, DF, "RibbonBScale(Special)", Param4]}."RibbonBScale(Special)", B=@n45)
   Alpha=ScalarParameter() {parameterName=UseBSpecialFix}
  )
  n47 = Add(A=@n34, B=Saturate(None=Multiply(A=@n44, B=@n46)))
  n48 = VectorParameter() {group=Opacity, parameterName=Opacity_EdgeMask_InvertControl, channelNames {a="", b=Power, g=Multiply, r=Offset}, defaultValue=<rgba 0 3 1 1>}
  n49 = VectorParameter() {group=Opacity, parameterName=Opacity_EdgeMask_Control, channelNames {a="", b=Power, g=Multiply, r=Offset}, defaultValue=<rgba 0 3 1 1>}
  n50 = Multiply(
   A=MaterialFunctionCall!(UorV=@n38.G, EdgeControl=@n39) {materialFunction=/Game/Effects/Niagara/MasterMaterials/Functions/Common/FXCommon_EdgeMask_Simple}.Mask
   B=Multiply(
    A=StaticSwitchParameter(True=@n47, False=@n34) {parameterName=UseBTexture}
    B=MaterialFunctionCall!(UorV=@n38.R, Invert_EdgeControl=@n48, NoInvert_EdgeControl=@n49) {materialFunction=/Game/Effects/Niagara/MasterMaterials/Functions/Common/FXCommon_EdgeMask_Ex}.Mask
   )
  )
  n51 = ScalarParameter() {defaultValue=2, parameterName=DF}
  n52 = Multiply(A=DynamicParameter() {paramNames [WPO, DF, "RibbonBScale(Special)", Param4]}.DF, B=@n51)
  n53 = ScalarParameter() {parameterName=UseDDF?}
  n54 = Power(Base=DepthFade(FadeDistance=LinearInterpolate(A=@n51, B=@n52, Alpha=@n53)), Exp=ScalarParameter() {defaultValue=2, parameterName=DFP})
  Multiply(
   A=Power(Base=Saturate(None=@n37), Exp=ScalarParameter() {defaultValue=1.5, parameterName=FresnelPow}) {constExponent=1}
   B=Multiply(A=@n36.A, B=Multiply(A=@n50, B=@n54))
  )
  n55 = MaterialFunctionCall(Float2=TextureCoordinate()) {materialFunction=&BreakOutFloat2Components}
  n56 = OneMinus(None=@n55.R)
  n57 = ScalarParameter() {parameterName=FlipU}
  NamedRerouteDeclaration(None=MaterialFunctionCall(X=@n55.G, Y=LinearInterpolate(A=@n55.R, B=@n56, Alpha=@n57)) {materialFunction=&MakeFloat2}.Result) {name=UV, nodeColor=<rgba 0.6641 0 1 1>}
  NamedRerouteDeclaration(None=Add(A=Time() {bOverride_Period=true, period=6666}, B=ScalarParameter() {parameterName=Ra})) {name=Timer, nodeColor=<rgba 0 1 0.4844 1>}
  n58 = Multiply(A=@n1.WPO) {constB=1.2}
  Power(Base=Saturate(None=Abs(None=@n58))) {constExponent=0.75}
 ]
}
